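# Sparks_GreenRm is a SoundClass for green room audio in the Sparks maps.
# It has no parent class and routes its sounds to the Fortnite_SFX submix by default.
# Reverb is turned off for sounds in this class.
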SoundClass /SparksMaps/Audio/GreenRoom/Sparks_GreenRm
parent=null
properties {properties {bReverb=false, defaultSubmix=&Fortnite_SFX}}
