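# Stereo looping MetaSound source for the base ambience bed of the Sparks green room map, assigned to the green room sound class with its own attenuation.
# On play, a wave player starts the ambience asset while an ADSR envelope fades it in over a slow attack to full sustain.
# The envelope drives the linear volume of a shared Fortnite stereo mixer patch, which feeds both output channels.
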MetaSoundSource /SparksMaps/Audio/GreenRoom/MSS_Pilg_GrRm_AmbBase
inputs [UE.Source.OnPlay {type=Trigger}]
interfaces [UE.OutputFormat.Stereo:1.0, UE.Source:1.1]
outputs [
 "UE.OutputFormat.Stereo.Audio:0" {type=Audio}
 "UE.OutputFormat.Stereo.Audio:1" {type=Audio}
]
pages [Default {
  input_defaults {UE.Source.OnPlay=null}
  nodes [
   ADSREnvelopeADSREnvelope_0 = ADSR Envelope.ADSR Envelope:Float(Trigger Attack=^UE.Source.OnPlay) {also {defaults {Attack Time=5, Sustain Level=1}}}
   WavePlayer_0 = UE.Wave Player:Stereo(Play=^UE.Source.OnPlay) {also {defaults {Loop="<MetaSoundBuilderResult.SUCCEEDED: 0>", Wave Asset=&/GreenRm_AmbBase}, version [1,1]}}
   StereoMixer_One_0 = /FortMetasound/Patches/StereoMixer_One("Volume(Lin)"=@ADSREnvelopeADSREnvelope_0.Out Envelope, InL=@WavePlayer_0.Out Left, InR=@WavePlayer_0.Out Right)
  ]
  links [{dst="@UE.OutputFormat.Stereo.Audio:0", src=StereoMixer_One_0.OutL}, {dst="@UE.OutputFormat.Stereo.Audio:1", src=StereoMixer_One_0.OutR}]
 }]
properties {
 assetImportData=&self:AssetImportData
 attenuationSettings=&/Pilg_GrRm_AmbBase_Attn
 bLooping=true
 duration=10000
 maxDistance=6000
 numChannels=2
 outputFormat=Stereo
 sampleRate=48000
 soundClassObject=&/Sparks_GreenRm
 volume=0.15
}
