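# Mono MetaSound source for the light hum ambience in the Sparks Green Room, set as a looping, attenuated sound.
# On play, a random pick from a wave asset array triggers a mono wave player that feeds the output.
# The shared RandomPitchTimeVolume patch supplies the player's pitch shift within a few semitones either way; only its pitch output is used.
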MetaSoundSource /SparksMaps/Audio/GreenRoom/MSS_Pilg_GrRm_LightHum
inputs [UE.Source.OnPlay {type=Trigger}]
interfaces [UE.OutputFormat.Mono:1.0, UE.Source:1.1]
outputs ["UE.OutputFormat.Mono.Audio:0" {type=Audio}]
pages [Default {
  input_defaults {UE.Source.OnPlay=null}
  nodes [
   InitVariable.WaveAsset:Array()
   ArrayRandomGet_0 = Array.Random Get:WaveAsset:Array(Next=^UE.Source.OnPlay, In Array=VariableAccessor.WaveAsset:Array(Variable=InitVariable.WaveAsset:Array().Variable).Value)
   WavePlayer_0 = UE.Wave Player:Mono(
    Play=@ArrayRandomGet_0.On Next
    Wave Asset=@ArrayRandomGet_0.Value
    Pitch Shift=/FortMetasound/Patches/MSP_RandomPitchTimeVolume() {also {defaults {"Random Pitch Max (Semitones)"=3, "Random Pitch Min (Semitones)"=-3}}}.Random Pitch
   ) {also {defaults {Loop="<MetaSoundBuilderResult.SUCCEEDED: 0>"}, version [1,1]}}
  ]
  links [{dst="@UE.OutputFormat.Mono.Audio:0", src=WavePlayer_0.Out Mono}]
 }]
properties {
 assetImportData=&self:AssetImportData
 attenuationSettings=&/Pilg_GrRm_LightHum_Attn
 bLooping=true
 duration=10000
 maxDistance=1075
 sampleRate=48000
 soundClassObject=&/Sparks_GreenRm
 volume=0.251
}
