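# Mono MetaSound source for spatialized static ambience in the Sparks Green Room, with its own attenuation and sound class.
# A repeating trigger with a randomized delay fires one-shot waves, randomly chosen from an array and randomly pitch-shifted.
# A second player starts a randomly chosen wave once on play as an unpitched layer.
# A two-input mixer combines both, with the second layer at reduced gain.
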MetaSoundSource /SparksMaps/Audio/GreenRoom/MSS_Pilg_GrRm_Static
inputs [
 In Array {type=WaveAsset:Array}
 UE.Source.OnPlay {type=Trigger}
]
interfaces [UE.OutputFormat.Mono:1.0, UE.Source:1.1]
outputs ["UE.OutputFormat.Mono.Audio:0" {type=Audio}]
pages [Default {
  input_defaults {UE.Source.OnPlay=null, In Array []}
  nodes [
   n1 = UE.RandomTime(Next=UE.TriggerRepeat(Start=^UE.Source.OnPlay) {also {defaults {Period=9}}}.RepeatOut) {also {defaults {Max=5, Min=0.1}}}
   TriggerDelay_0 = UE.Trigger Delay(In=@n1.On Next, Delay Time=@n1.Value)
   n2 = UE.RandomFloat(Next=@TriggerDelay_0.Out) {also {defaults {Max=3, Min=-3}}}
   ArrayRandomGet_0 = Array.Random Get:WaveAsset:Array(Next=@n2.On Next, In Array=VariableAccessor.WaveAsset:Array(Variable=InitVariable.WaveAsset:Array().Variable).Value) {also {defaults {No Repeats=0}}}
   WavePlayer_0 = UE.Wave Player:Mono(Play=@ArrayRandomGet_0.On Next, Wave Asset=@ArrayRandomGet_0.Value, Pitch Shift=UE.Reroute(Value=@n2.Value).Value) {also {version [1,1]}}
   ArrayRandomGet_1 = Array.Random Get:WaveAsset:Array(Next=^UE.Source.OnPlay, In Array=VariableAccessor.WaveAsset:Array(Variable=InitVariable.WaveAsset:Array().Variable).Value)
   WavePlayer_1 = UE.Wave Player:Mono(Play=@ArrayRandomGet_1.On Next, Wave Asset=@ArrayRandomGet_1.Value) {also {defaults {Loop="<MetaSoundBuilderResult.SUCCEEDED: 0>"}, version [1,1]}}
   AudioMixerAudioMixerMono2_0 = "AudioMixer.Audio Mixer (Mono, 2)"(In 0=@WavePlayer_0.Out Mono, In 1=@WavePlayer_1.Out Mono) {also {defaults {Gain 1=0.6}}}
  ]
  links [{dst="@UE.OutputFormat.Mono.Audio:0", src=AudioMixerAudioMixerMono2_0.Out}]
 }]
properties {
 assetImportData=&self:AssetImportData
 attenuationSettings=&/Pilg_GrRm_Static_Attn
 bLooping=true
 duration=10000
 maxDistance=2000
 sampleRate=48000
 soundClassObject=&/Sparks_GreenRm
 volume=0.1
}
