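# Stereo MetaSound source for the far crowd ambience loop in the Pilg green room of SparksMaps.
# On play, a single stereo Wave Player starts GreenRm_crowdlp and feeds both output channels directly, with no other processing in the graph.
# The source is flagged as looping, plays at reduced volume and routes to the Sparks_GreenRm sound class.
# The player's Loop default holds a stray builder-result string rather than a boolean.
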MetaSoundSource /SparksMaps/Audio/GreenRoom/MSS_Pilg_GrnRm_CrowdFar
inputs [UE.Source.OnPlay {type=Trigger}]
interfaces [UE.OutputFormat.Stereo:1.0, UE.Source:1.1]
outputs [
 "UE.OutputFormat.Stereo.Audio:0" {type=Audio}
 "UE.OutputFormat.Stereo.Audio:1" {type=Audio}
]
pages [Default {
  input_defaults {UE.Source.OnPlay=null}
  nodes [
   InitVariable.WaveAsset:Array()
   WavePlayer_0 = UE.Wave Player:Stereo(Play=^UE.Source.OnPlay) {also {defaults {Loop="<MetaSoundBuilderResult.SUCCEEDED: 0>", Wave Asset=&/GreenRm_crowdlp}, version [1,1]}}
  ]
  links [{dst="@UE.OutputFormat.Stereo.Audio:0", src=WavePlayer_0.Out Left}, {dst="@UE.OutputFormat.Stereo.Audio:1", src=WavePlayer_0.Out Right}]
 }]
properties {
 assetImportData=&self:AssetImportData
 bLooping=true
 duration=10000
 numChannels=2
 outputFormat=Stereo
 sampleRate=48000
 soundClassObject=&/Sparks_GreenRm
 volume=0.6
}
