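# Mono MetaSound source for the portal shake sound in the Pilg Green Room audio set.
# On play, a single Wave Player starts the GreenRm_PortalShake wave and feeds it straight to the mono output.
# The source is flagged as looping, uses its own portal shake attenuation asset for distance falloff, and is assigned to the Sparks_GreenRm sound class.
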MetaSoundSource /SparksMaps/Audio/GreenRoom/MSS_Pilg_GrnRm_PortalShake
inputs [UE.Source.OnPlay {type=Trigger}]
interfaces [UE.OutputFormat.Mono:1.0, UE.Source:1.1]
outputs ["UE.OutputFormat.Mono.Audio:0" {type=Audio}]
pages [Default {
  input_defaults {UE.Source.OnPlay=null}
  nodes [WavePlayer_0 = UE.Wave Player:Mono(Play=^UE.Source.OnPlay) {also {defaults {Loop="<MetaSoundBuilderResult.SUCCEEDED: 0>", Wave Asset=&/GreenRm_PortalShake}, version [1,1]}}]
  links [{dst="@UE.OutputFormat.Mono.Audio:0", src=WavePlayer_0.Out Mono}]
 }]
properties {
 assetImportData=&self:AssetImportData
 attenuationSettings=&/Pilg_GrRm_PortalShake_Attn
 bLooping=true
 duration=10000
 maxDistance=1400
 sampleRate=48000
 soundClassObject=&/Sparks_GreenRm
 volume=1.2
}
