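# Sound cue for the song-select drum roll UI sound in the Sparks GreenRoom audio set.
# A single wave player feeds the UI_SongSelect_DrumRoll wave straight to the output, with no randomization or mixing.
# The cue boosts volume above unity and routes through the Fort_SFX_InGame sound class.
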SoundCue /SparksMaps/Audio/GreenRoom/UI_SongSelectDrumRoll_Cue
root=WavePlayer_0
nodes [{id=WavePlayer_0, properties {soundWaveAssetPtr=&/UI_SongSelect_DrumRoll}}]
properties {maxDistance=2097152, soundClassObject=&Fort_SFX_InGame, volumeMultiplier=1.5}
